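; MatchMist flying orb actor that opens its hatch on a calendar event and grants players a radius-marking scan ability.
; BeginPlay sets the death-beam flag, then after calendar sync watches an event and increments ChangeHatchMesh following a delay.
; Its OnRep swaps in the open hatch mesh, and SetHatchMesh is a test cheat bypassing the calendar.
; HandlePlayerAbilitySet removes any existing ability set, then re-equips it, tracking handles per controller in a map.
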(event EventBeginPlay
  (|CallParentFunction)
  (Variables|Default|SetHF_DeathBeam_Enabled)
  (MCPContext|InitialCalendarSyncCompleted
    (:OnSyncComplete
      (bind _eventname (MCPContext|WatchCalendarEvent (Variables|Default|GetEvent%20to%20Watch))
        (:OnEventActive
          (Fortnite|FortShippingLog 0 (Utilities|String|Append "[RoundFork] Orb has received Stage " _eventname " to change the hatch mesh"))
          (Utilities|FlowControl|Delay 3.0)
          (Variables|Default|SetChangeHatchMesh (+ (Variables|Default|GetChangeHatchMesh) 1)))))))

(event Custom|SetHatchMesh
  ;[0.07752 0.6406 0.07007 0.5] Cheats for Testing w/o MCP Calendar
  (Variables|Default|SetChangeHatchMesh (+ (Variables|Default|GetChangeHatchMesh) 1)))

(fn HandlePlayerAbilitySet
  (params (FortPlayerControllerAthena :type Fort%20Player%20Controller%20Athena%20Object%20Reference))
  (bind _fortplayercontrollerathena (Variables|GetFortPlayerControllerAthena))
  (bind _as_radiusmarking_map (Variables|Default|GetAS_RadiusMarking_Map))
  (bind (_value _returnvalue) (Utilities|Map|Find _as_radiusmarking_map _fortplayercontrollerathena))
  (Utilities|FlowControl|Sequence
    (:then_0
      (if (Ability|GameplayAbility|HasAnyAbilitieswithAssetTag (Ability|GetAbilitySystemComponent (Variables|GetFortPlayerControllerAthena)) "(TagName=\"Ability.Orb.RadiusMarking\")")
        (if _returnvalue
          (Utilities|Map|Remove (Variables|Default|GetAS_RadiusMarking_Map) (Variables|GetFortPlayerControllerAthena))
          (Ability|UnequipFortAbilitySet _value)
          (Fortnite|FortShippingLog 0 (Utilities|String|Append "BGA_DCN_OrbActor: Unequipping ability for player " (Utilities|GetDisplayName (Variables|GetFortPlayerControllerAthena))))
          (else
            (Fortnite|FortShippingLog 0 "BGA_DCN_OrbActor: Trying to remove the ability set, but it doesn't exist on the map.")))))
    (:then_1
      (bind _returnvalue_2 (Ability|EquipFortAbilitySet (Variables|GetFortPlayerControllerAthena) "/DragonCartNarrative/MatchMist/AS_DCN_OrbActor_RadiusMarking_MatchMist.AS_DCN_OrbActor_RadiusMarking_MatchMist"))
      (Utilities|Map|Add (Variables|Default|GetAS_RadiusMarking_Map) (Variables|GetFortPlayerControllerAthena) _returnvalue_2)
      (Development|PrintString (Utilities|String|Append "BGA_DCN_OrbActor: " (Utilities|GetDisplayName (Variables|GetFortPlayerControllerAthena)) " HAS SCAN ABILITY") true true "(R=0.000000,G=0.660000,B=1.000000,A=1.000000)" 1e+01)
      (return))))

(fn OnRep_ChangeHatchMesh
  (Components|StaticMesh|SetStaticMesh (Variables|Default|GetOrbHatch) "/DragonCartNarrative/Assets/FlyingOrb/Meshes/Static/SM_FlyingOrb_Open.SM_FlyingOrb_Open"))

(fn ConstructionScript
  (|CallParentFunction))
