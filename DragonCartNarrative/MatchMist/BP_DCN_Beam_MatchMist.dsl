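; Tractor beam actor that gives overlapping player pawns low gravity and vortex movement through nested cone triggers, handled on the server.
; The big cone applies fall-damage immunity, the tractor beam effect and beam FX, removing the latter two on exit.
; The small cone, while the pawn is also in the big cone, swaps the tractor effect for vortex state and tractor FX, both cleared on exit.
; A separate ApplyTeleportGA helper applies a teleport effect; OuterRadius handlers are empty.
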(event EventBeginPlay)

(event EventTick
  (params (DeltaSeconds :type Float%20%28single-precision%29)))

(event OnComponentBeginOverlap(OuterRadius)
  (params (OverlappedComponent :type Primitive%20Component%20Object%20Reference) (OtherActor :type Actor%20Object%20Reference) (OtherComp :type Primitive%20Component%20Object%20Reference) (OtherBodyIndex :type Integer) (bFromSweep :type Boolean) (SweepResult :type Hit%20Result%20Structure%20%28by%20ref%29)))

(event OnComponentEndOverlap(OuterRadius)
  (params (OverlappedComponent :type Primitive%20Component%20Object%20Reference) (OtherActor :type Actor%20Object%20Reference) (OtherComp :type Primitive%20Component%20Object%20Reference) (OtherBodyIndex :type Integer)))

(event Custom|SetExitingPlayersIntoVortex
  (bind _asfort_player_pawn_athena (Utilities|Casting|CastToFortPlayerPawnAthena (Variables|Default|GetOther%20Actor)))
  (Fortnite|FortShippingLog self "Player pawn has overlapped Inner Beam and got low gravity")
  (Utilities|IsValid (Ability|GetFortAbilitySystemComponent _asfort_player_pawn_athena)
    (:Is%20Valid
      (GameplayEffects|FortApplyGameplayEffectToSelf :self (Ability|GetFortAbilitySystemComponent _asfort_player_pawn_athena) :GameplayEffectClass "/Game/Athena/Items/Parents/FallImmunity/GE_Athena_TillLandFallDamageImmunity_NoFX_Grant_ApplyInAirOnly.GE_Athena_TillLandFallDamageImmunity_NoFX_Grant_ApplyInAirOnly_C" :Level 0.0)
      (GameplayEffects|FortApplyGameplayEffectToSelf :self (Ability|GetFortAbilitySystemComponent _asfort_player_pawn_athena) :GameplayEffectClass (Variables|Default|GetTractorBeamEffect) :Level 0.0)
      (GameplayEffects|ApplyGameplayEffectToTarget :self (Ability|GetFortAbilitySystemComponent _asfort_player_pawn_athena) :GameplayEffectClass "/DragonCartNarrative/Gameplay/Effect/GE_DCN_BeamFX.GE_DCN_BeamFX_C" :Target (Ability|GetFortAbilitySystemComponent _asfort_player_pawn_athena) :Level 0.0)
      (Variables|/Script/FortniteGame.FortPawn|SetbIgnoreNextFallingDamage true _asfort_player_pawn_athena))))

(event OnComponentBeginOverlap(Big_Cone_Collision)
  (params (OverlappedComponent :type Primitive%20Component%20Object%20Reference) (OtherActor :type Actor%20Object%20Reference) (OtherComp :type Primitive%20Component%20Object%20Reference) (OtherBodyIndex :type Integer) (bFromSweep :type Boolean) (SweepResult :type Hit%20Result%20Structure%20%28by%20ref%29))
  (Networking|SwitchHasAuthority
    (:Authority
      (Variables|Default|SetOther%20Actor OtherActor)
      (bind _asfort_player_pawn_athena (Utilities|Casting|CastToFortPlayerPawnAthena (Variables|Default|GetOther%20Actor)))
      (Fortnite|FortShippingLog self "Player pawn has overlapped Inner Beam and got low gravity")
      (Utilities|IsValid (Ability|GetFortAbilitySystemComponent _asfort_player_pawn_athena)
        (:Is%20Valid
          (GameplayEffects|FortApplyGameplayEffectToSelf :self (Ability|GetFortAbilitySystemComponent _asfort_player_pawn_athena) :GameplayEffectClass "/Game/Athena/Items/Parents/FallImmunity/GE_Athena_TillLandFallDamageImmunity_NoFX_Grant_ApplyInAirOnly.GE_Athena_TillLandFallDamageImmunity_NoFX_Grant_ApplyInAirOnly_C" :Level 0.0)
          (GameplayEffects|FortApplyGameplayEffectToSelf :self (Ability|GetFortAbilitySystemComponent _asfort_player_pawn_athena) :GameplayEffectClass (Variables|Default|GetTractorBeamEffect) :Level 0.0)
          (GameplayEffects|ApplyGameplayEffectToTarget :self (Ability|GetFortAbilitySystemComponent _asfort_player_pawn_athena) :GameplayEffectClass "/DragonCartNarrative/Gameplay/Effect/GE_DCN_BeamFX.GE_DCN_BeamFX_C" :Target (Ability|GetFortAbilitySystemComponent _asfort_player_pawn_athena) :Level 0.0)
          (Variables|/Script/FortniteGame.FortPawn|SetbIgnoreNextFallingDamage true _asfort_player_pawn_athena))))))

(event OnComponentEndOverlap(Big_Cone_Collision)
  (params (OverlappedComponent :type Primitive%20Component%20Object%20Reference) (OtherActor :type Actor%20Object%20Reference) (OtherComp :type Primitive%20Component%20Object%20Reference) (OtherBodyIndex :type Integer))
  (Networking|SwitchHasAuthority
    (:Authority
      (Variables|Default|SetOther%20Actor OtherActor)
      (bind _asfort_player_pawn_athena (Utilities|Casting|CastToFortPlayerPawnAthena (Variables|Default|GetOther%20Actor)))
      (Fortnite|FortShippingLog self "Player pawn has left the Beam")
      (Utilities|IsValid (Ability|GetFortAbilitySystemComponent _asfort_player_pawn_athena)
        (:Is%20Valid
          (GameplayEffects|RemoveActiveGameplayEffectbySourceEffect (Ability|GetFortAbilitySystemComponent _asfort_player_pawn_athena) (Variables|Default|GetTractorBeamEffect) 0 -1)
          (GameplayEffects|RemoveActiveGameplayEffectbySourceEffect (Ability|GetFortAbilitySystemComponent _asfort_player_pawn_athena) "/DragonCartNarrative/Gameplay/Effect/GE_DCN_BeamFX.GE_DCN_BeamFX_C" 0 -1)
          (Variables|/Script/FortniteGame.FortPawn|SetbIgnoreNextFallingDamage true _asfort_player_pawn_athena))))))

(event OnComponentBeginOverlap(Small_Cone_Collision)
  (params (OverlappedComponent :type Primitive%20Component%20Object%20Reference) (OtherActor :type Actor%20Object%20Reference) (OtherComp :type Primitive%20Component%20Object%20Reference) (OtherBodyIndex :type Integer) (bFromSweep :type Boolean) (SweepResult :type Hit%20Result%20Structure%20%28by%20ref%29))
  (Networking|SwitchHasAuthority
    (:Authority
      (bind _output_get (Variables|Default|SetOther%20Actor OtherActor))
      (bind _asfort_player_pawn_athena (Utilities|Casting|CastToFortPlayerPawnAthena _output_get))
      (Fortnite|FortShippingLog self "Overlapping the High Capsule")
      (if (Collision|IsOverlappingActor (Variables|Default|GetBig_Cone_Collision) _output_get :!class /Script/Engine.PrimitiveComponent)
        (Fortnite|FortShippingLog self "Still in Beam")
        (Vortex|SetinVortex _asfort_player_pawn_athena true 0.0 5e+02 5e+03 -2.0 5e+01 false)
        (GameplayEffects|RemoveActiveGameplayEffectbySourceEffect (Ability|GetFortAbilitySystemComponent _asfort_player_pawn_athena) (Variables|Default|GetTractorBeamEffect) 0 -1)
        (GameplayEffects|FortApplyGameplayEffectToSelf :self (Ability|GetFortAbilitySystemComponent _asfort_player_pawn_athena) :GameplayEffectClass "/DragonCartNarrative/Gameplay/Effect/GE_DCN_TractorBeamFX.GE_DCN_TractorBeamFX_C" :Level 0.0)
        (else
          (Fortnite|FortShippingLog self "NOT in Beam")
          (GameplayEffects|RemoveActiveGameplayEffectbySourceEffect (Ability|GetFortAbilitySystemComponent _asfort_player_pawn_athena) "/DragonCartNarrative/Gameplay/Effect/GE_DCN_TractorBeamFX.GE_DCN_TractorBeamFX_C" 0 -1))))))

(event OnComponentEndOverlap(Small_Cone_Collision)
  (params (OverlappedComponent :type Primitive%20Component%20Object%20Reference) (OtherActor :type Actor%20Object%20Reference) (OtherComp :type Primitive%20Component%20Object%20Reference) (OtherBodyIndex :type Integer))
  (Networking|SwitchHasAuthority
    (:Authority
      (bind _output_get (Variables|Default|SetOther%20Actor OtherActor))
      (bind _asfort_player_pawn_athena (Utilities|Casting|CastToFortPlayerPawnAthena _output_get))
      (Fortnite|FortShippingLog self "Players has left the high capsule")
      (GameplayEffects|RemoveActiveGameplayEffectbySourceEffect (Ability|GetFortAbilitySystemComponent _asfort_player_pawn_athena) "/DragonCartNarrative/Gameplay/Effect/GE_DCN_TractorBeamFX.GE_DCN_TractorBeamFX_C" 0 -1)
      (Vortex|SetinVortex _asfort_player_pawn_athena false 0.0 0.0 0.0 0.0 0.0 false))))

(fn ApplyTeleportGA
  (params (PlayerToTeleport :type Fort%20Player%20Pawn%20Athena%20Object%20Reference))
  (bind _returnvalue (Ability|GetFortAbilitySystemComponent PlayerToTeleport))
  (Utilities|IsValid _returnvalue
    (:Is%20Valid
      (GameplayEffects|FortApplyGameplayEffectToTarget :self _returnvalue :GameplayEffectClass (Variables|Default|GetTeleportEffect) :TargetOfEffect _returnvalue :OptionalSourceObject self :OptionalEffectCauser self :Level 0.0))))

(fn OnRep_TargetRift)

(fn ConstructionScript)
