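; Data layer manager for the DragonCartNarrative MatchMist content that switches World Partition data layers as RoundFork calendar events arrive.
; After the initial MCP calendar sync, each active watched event is looked up in a calendar-to-layer-name table.
; Matching names are resolved to layer instances by short name, then set Activated or Unloaded.
; Test events fake the rift and orb stages without the calendar, most then driving the rift or orb actor after a delay.
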(event EventBeginPlay
  (MCPContext|InitialCalendarSyncCompleted
    (:OnSyncComplete
      (bind _eventname (MCPContext|WatchCalendarEvents (Variables|Default|GetEvents%20to%20Watch))
        (:OnEventActive
          (Fortnite|FortShippingLog 0 (Utilities|String|Append "[RoundFork]" "" _eventname " " "is received from calendar"))
          (CallFunction|HandleDataLayerSelectionbyEvent _eventname)
          (CallFunction|GetDataLayersinUEFNMap)
          (CallFunction|ActivateandUnloadDataLayers))))))

(event Custom|Test_RiftIn1
  ;[0.04507 0.4427 0.0415 0.5] Cheats for Testing w/o MCP Calendar
  (Fortnite|FortShippingLog 0 "[RoundFork] Start Event: RiftIn1")
  (CallFunction|HandleDataLayerSelectionbyEvent "RF_RiftIn1")
  (CallFunction|GetDataLayersinUEFNMap)
  (CallFunction|ActivateandUnloadDataLayers)
  (Utilities|FlowControl|Delay 2.0)
  (bind _outactors (Actor|GetAllActorsOfClass "/DragonCartNarrative/MatchMist/BP_DCN_Rift_Actor_MatchMist.BP_DCN_Rift_Actor_MatchMist_C"))
  (Class|BPDCNRiftActorMatchMist|Riftin1 (Utilities|Array|Get(acopy) _outactors)))

(event Custom|Test_RiftIn2
  ;[0.04507 0.4427 0.0415 0.5] Cheats for Testing w/o MCP Calendar
  (Fortnite|FortShippingLog 0 "[RoundFork] Start Event: RiftIn2")
  (CallFunction|HandleDataLayerSelectionbyEvent "RF_RiftIn2")
  (CallFunction|GetDataLayersinUEFNMap)
  (CallFunction|ActivateandUnloadDataLayers)
  (Utilities|FlowControl|Delay 2.0)
  (bind _outactors (Actor|GetAllActorsOfClass "/DragonCartNarrative/MatchMist/BP_DCN_Rift_Actor_MatchMist.BP_DCN_Rift_Actor_MatchMist_C"))
  (Class|BPDCNRiftActorMatchMist|Riftin2 (Utilities|Array|Get(acopy) _outactors)))

(event Custom|Test_OrbIn1
  ;[0.04507 0.4427 0.0415 0.5] Cheats for Testing w/o MCP Calendar
  (Fortnite|FortShippingLog 0 "[RoundFork] Start Event: OrbIn1")
  (CallFunction|HandleDataLayerSelectionbyEvent "RF_OrbIn1")
  (CallFunction|GetDataLayersinUEFNMap)
  (CallFunction|ActivateandUnloadDataLayers)
  (Utilities|FlowControl|Delay 2.0)
  (bind _outactors (Actor|GetAllActorsOfClass "/DragonCartNarrative/MatchMist/BP_DCN_Rift_Actor_MatchMist.BP_DCN_Rift_Actor_MatchMist_C"))
  (Class|BPDCNRiftActorMatchMist|Orbin1 (Utilities|Array|Get(acopy) _outactors)))

(event Custom|Test_OrbIn2
  ;[0.04507 0.4427 0.0415 0.5] Cheats for Testing w/o MCP Calendar
  (Fortnite|FortShippingLog 0 "[RoundFork] Start Event: OrbIn2")
  (CallFunction|HandleDataLayerSelectionbyEvent "RF_OrbIn2")
  (CallFunction|GetDataLayersinUEFNMap)
  (CallFunction|ActivateandUnloadDataLayers)
  (Utilities|FlowControl|Delay 2.0)
  (bind _outactors (Actor|GetAllActorsOfClass "/DragonCartNarrative/MatchMist/BGA_DCN_OrbActor_MatchMist.BGA_DCN_OrbActor_MatchMist_C"))
  (Class|BGADCNOrbActorMatchMist|SetHatchMesh (Utilities|Array|Get(acopy) _outactors)))

(event Custom|Test_OrbOut1
  ;[0.04507 0.4427 0.0415 0.5] Cheats for Testing w/o MCP Calendar
  (Fortnite|FortShippingLog 0 "[RoundFork] Start Event: OrbOut1")
  (CallFunction|HandleDataLayerSelectionbyEvent "RF_OrbOut1")
  (CallFunction|GetDataLayersinUEFNMap)
  (CallFunction|ActivateandUnloadDataLayers)
  (Utilities|FlowControl|Delay 2.0)
  (bind _outactors (Actor|GetAllActorsOfClass "/DragonCartNarrative/MatchMist/BP_DCN_Rift_Actor_MatchMist.BP_DCN_Rift_Actor_MatchMist_C"))
  (Class|BPDCNRiftActorMatchMist|OrbOut1 (Utilities|Array|Get(acopy) _outactors)))

(event Custom|Test_OrbOut2
  ;[0.04507 0.4427 0.0415 0.5] Cheats for Testing w/o MCP Calendar
  (Fortnite|FortShippingLog 0 "[RoundFork] Start Event: OrbOut2")
  (CallFunction|HandleDataLayerSelectionbyEvent "RF_OrbOut2")
  (CallFunction|GetDataLayersinUEFNMap)
  (CallFunction|ActivateandUnloadDataLayers))

(fn ActivateAndUnloadDataLayers
  (bind _returnvalue (WorldPartition|GetDataLayerManager))
  (Utilities|IsValid _returnvalue
    (:Is%20Valid
      _)
    (:Is%20Not%20Valid
      (Fortnite|FortShippingLog self "[RoundFork] Cannot find Data Layer Manager when Trying to Activate/Unload DLs.")))
  (for _array_element (Variables|Default|GetDataLayersToActivate)
    (Fortnite|FortShippingLog self (Utilities|String|Append "[RoundFork] Found and Activate DLs: " (DataLayer|GetDataLayerShortName _array_element)))
    (DataLayers|SetDataLayerInstanceRuntimeState _returnvalue _array_element "Activated" true))
  (for _array_element_2 (Variables|Default|GetDataLayersToUnload)
    (Fortnite|FortShippingLog self (Utilities|String|Append "[RoundFork] Found and Unload DLs: " (DataLayer|GetDataLayerShortName _array_element_2)))
    (DataLayers|SetDataLayerInstanceRuntimeState _returnvalue _array_element_2 "Unloaded" true)))

(fn GetDataLayersInUEFNMap
  (Utilities|Array|Clear (Variables|Default|GetDataLayersToActivate))
  (Utilities|Array|Clear (Variables|Default|GetDataLayersToUnload))
  (bind _returnvalue (WorldPartition|GetDataLayerManager))
  (Utilities|IsValid _returnvalue
    (:Is%20Valid
      (for _array_element (DataLayers|GetDataLayerInstances _returnvalue)
        (if (Utilities|Array|ContainsItem (Variables|Default|GetNames_DataLayersToActivate) (Utilities|String|StringToName (DataLayer|GetDataLayerShortName _array_element)))
          (Utilities|Array|Add (Variables|Default|GetDataLayersToActivate) _array_element)
          (else
            (if (Utilities|Array|ContainsItem (Variables|Default|GetNames_DataLayersToUnload) (Utilities|String|StringToName (DataLayer|GetDataLayerShortName _array_element)))
              (Utilities|Array|Add (Variables|Default|GetDataLayersToUnload) _array_element))))))
    (:Is%20Not%20Valid
      (Fortnite|FortShippingLog 0 "Failed to find the DL Manager in UEFN"))))

(fn HandleDataLayerSelectionByEvent
  (params (EventName :type String))
  (for _array_element (Variables|Default|GetCalendarDataLayerName)
    (Fortnite|FortShippingLog 0 (Utilities|String|Append "[RoundFork] EventStage:  " EventName))
    (if (Utilities|String|EqualExactly(String) EventName (Utilities|Struct|BreakCalendarDataLayerName _array_element))
      (bind _output_get (Variables|Default|SetNames_DataLayersToActivate (.ActivatedDataLayerName_9_7E617B164F789C5872C4DD92614365BF (Utilities|Struct|BreakCalendarDataLayerName _array_element))))
      (for _array_element_2 _output_get
        (Fortnite|FortShippingLog 0 (Utilities|String|Append "[RoundFork] Attempt to Activate DL: " (Utilities|String|ToString(Name) _array_element_2))))
      (bind _output_get_2 (Variables|Default|SetNames_DataLayersToUnload (.UnloadedDataLayerName_10_2D760D01407A4DD0BCF7A8A4099131F8 (Utilities|Struct|BreakCalendarDataLayerName _array_element))))
      (for _array_element_3 _output_get_2
        (Fortnite|FortShippingLog 0 (Utilities|String|Append "Attepmt to Unload DL: " (Utilities|String|ToString(Name) _array_element_3)))))))

(fn ConstructionScript)
