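; Rift actor for DragonCart narrative matches that sizes a rift and its looping audio according to live calendar events.
; After calendar sync, each active event waits briefly, then looks up a scale and an audio attenuation value in maps keyed by event name.
; UpdateMPC's RepNotify sets RiftScale in a material parameter collection off-server and scales the build-blocking collision.
; Custom events simulate named calendar phases or set the size directly for testing.
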(event EventBeginPlay
  ; Set mpc setting based on calendar event
  (|CallParentFunction)
  (MCPContext|InitialCalendarSyncCompleted
    (:OnSyncComplete
      (bind _eventname (MCPContext|WatchCalendarEvents (Variables|Default|GetCalendar%20-%20RotatingOrbCalendars))
        (:OnEventActive
          (Fortnite|FortShippingLog 0 (Utilities|String|Append "[RoundFork]" "" _eventname " " "Is Active"))
          (Utilities|FlowControl|Delay 3.0)
          (Variables|Default|SetUpdateMPC (Utilities|Map|Find (Variables|Default|Get%20RotatingOrbCalendars%20Map) _eventname))
          (CallFunction|UpdateAudioAttenuationSettings (Utilities|Map|Find (Variables|Default|GetOrbCalendarsAttenuationMap) _eventname)))))))

(event Custom|RiftIn1
  ;[0.01221 0.3906 0.06567 0.5] Cheats for Testing w/o MCP Calendar
  (Variables|Default|SetUpdateMPC (Utilities|Map|Find (Variables|Default|Get%20RotatingOrbCalendars%20Map) "RF_RiftIn1"))
  (CallFunction|UpdateAudioAttenuationSettings (Utilities|Map|Find (Variables|Default|GetOrbCalendarsAttenuationMap) "RF_RiftIn1")))

(event Custom|RiftIn2
  ;[0.01221 0.3906 0.06567 0.5] Cheats for Testing w/o MCP Calendar
  (Variables|Default|SetUpdateMPC (Utilities|Map|Find (Variables|Default|Get%20RotatingOrbCalendars%20Map) "RF_RiftIn2"))
  (CallFunction|UpdateAudioAttenuationSettings (Utilities|Map|Find (Variables|Default|GetOrbCalendarsAttenuationMap) "RF_RiftIn2")))

(event Custom|OrbIn1
  ;[0.01221 0.3906 0.06567 0.5] Cheats for Testing w/o MCP Calendar
  (Variables|Default|SetUpdateMPC (Utilities|Map|Find (Variables|Default|Get%20RotatingOrbCalendars%20Map) "RF_OrbIn1"))
  (CallFunction|UpdateAudioAttenuationSettings (Utilities|Map|Find (Variables|Default|GetOrbCalendarsAttenuationMap) "RF_OrbIn1")))

(event Custom|OrbOut1
  ;[0.01221 0.3906 0.06567 0.5] Cheats for Testing w/o MCP Calendar
  (Variables|Default|SetUpdateMPC (Utilities|Map|Find (Variables|Default|Get%20RotatingOrbCalendars%20Map) "RF_OrbOut1"))
  (CallFunction|UpdateAudioAttenuationSettings (Utilities|Map|Find (Variables|Default|GetOrbCalendarsAttenuationMap) "RF_OrbOut1")))

(event Custom|SetRiftSize
  (params (UpdateMPC :type Float%20%28double-precision%29))
  ; Testing Only
  (Variables|Default|SetUpdateMPC UpdateMPC))

(event Custom|UpdateAudioAttenuationSettings
  (params (Value :type Float%20%28double-precision%29))
  (bind _riftloopaudio (Variables|Default|GetRiftLoopAudio))
  (bind _outattenuationsettings (Audio|Components|Audio|GetAttenuationSettingsToApply _riftloopaudio))
  (bind _output_get (Variables|/Script/Engine.AudioComponent|SetAttenuationOverrides _outattenuationsettings _riftloopaudio))
  (Variables|/Script/Engine.AudioComponent|SetbOverrideAttenuation true _riftloopaudio)
  (bind _structout (Utilities|Struct|SetmembersinSoundAttenuationSettings _output_get Value))
  (Audio|Components|Audio|AdjustAttenuation _riftloopaudio _structout))

(fn OnRep_UpdateMPC
  (bind _updatempc (Variables|Default|GetUpdateMPC))
  (if (not (Networking|IsServer))
    (Rendering|Material|SetScalarParameterValue "/DragonCartNarrative/Assets/VFX/RiftPortalIdle/MPC_DragonCart_Orb_Rift.MPC_DragonCart_Orb_Rift" "RiftScale" _updatempc :!class /Script/Engine.KismetMaterialLibrary))
  (Transformation|SetWorldScale3D (Variables|Default|GetBlockBuildingCollision) (Math|Conversions|ToVector(Float) _updatempc)))

(fn ConstructionScript)
