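; Cosmetic turret for the MatchMist event that shoots Niagara projectile volleys at the Orb actor once a calendar event is active and the Warmup phase step arrives.
; Setup aims the turret and derives the impact point on the scaled Orb's surface plus the projectile travel time.
; Firing runs only on non-authority machines, with paired looping timers: one for shot feedback, one delayed by travel time for impact sounds and Orb ripples.
; DoN counters pause between volleys and end in overheat.
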(event Custom|Test_OrbIn2
  (Fortnite|FortShippingLog 0 "[RoundFork] Turrets got the cheat and are about to shoot")
  (Utilities|FlowControl|Delay 1.0)
  (CallFunction|TurretProjectileSetup (Variables|Default|GetOrbActor))
  (bind _returnvalue (GameplayEventRouter|GetEventRouter "/Script/FortniteGame.FortGameState" self))
  (bind (_asynctaskproxy _payload) (GameplayEventRouter|StartListeningToStatefulEvent _returnvalue "/Script/FortniteGame.FortGamePhaseStepUpdatedEvent")
    (:then
      (Variables|Default|SetAsync%20Task%20Ref _asynctaskproxy))
    (:OnEventReceived
      _)
    (:OnSavedEventStateExists
      _))
  (switch Utilities|FlowControl|Switch|SwitchonEAthenaGamePhaseStep (Utilities|Struct|BreakFortGamePhaseStepUpdatedEvent _payload)
    (:Warmup
      (Utilities|FlowControl|Delay (+ (ScalableFloat|GetValueatLevel (Variables|Default|GetShootingDelayTime)) (Variables|Default|GetFiringStartOffset)))
      (CallFunction|StartFiring)
      (GameplayEventRouter|StopListeningtoStatefulEvent (Variables|Default|GetAsync%20Task%20Ref)))))

(event Custom|TestFiring
  (CallFunction|FindOrbActor)
  (CallFunction|TurretProjectileSetup (Variables|Default|GetOrbActor))
  (Utilities|FlowControl|Delay (+ (ScalableFloat|GetValueatLevel (Variables|Default|GetShootingDelayTime)) (Variables|Default|GetFiringStartOffset)))
  (CallFunction|StartFiring)
  (GameplayEventRouter|StopListeningtoStatefulEvent (Variables|Default|GetAsync%20Task%20Ref)))

(event EventBeginPlay
  (Class|FortAsyncActionCurrentPlaylistReady|PlaylistDataReadyAsync
    (:OnReady
      (CallFunction|FindOrbActor)
      (MCPContext|InitialCalendarSyncCompleted
        (:OnSyncComplete
          (MCPContext|WatchCalendarEvent (Variables|Default|GetEvent%20to%20Watch)
            (:OnEventActive
              (Utilities|FlowControl|Delay 1.0)
              (CallFunction|TurretProjectileSetup (Variables|Default|GetOrbActor))
              (bind _returnvalue (GameplayEventRouter|GetEventRouter "/Script/FortniteGame.FortGameState" self))
              (bind (_asynctaskproxy _payload) (GameplayEventRouter|StartListeningToStatefulEvent _returnvalue "/Script/FortniteGame.FortGamePhaseStepUpdatedEvent")
                (:then
                  (Variables|Default|SetAsync%20Task%20Ref _asynctaskproxy))
                (:OnEventReceived
                  _)
                (:OnSavedEventStateExists
                  _))
              (switch Utilities|FlowControl|Switch|SwitchonEAthenaGamePhaseStep (Utilities|Struct|BreakFortGamePhaseStepUpdatedEvent _payload)
                (:Warmup
                  (Utilities|FlowControl|Delay (+ (ScalableFloat|GetValueatLevel (Variables|Default|GetShootingDelayTime)) (Variables|Default|GetFiringStartOffset)))
                  (CallFunction|StartFiring)
                  (GameplayEventRouter|StopListeningtoStatefulEvent (Variables|Default|GetAsync%20Task%20Ref)))))))))))

(event Custom|StartFiring
  (bind _volleysbeforeoverheat (Variables|Default|GetVolleysBeforeOverheat))
  (bind _returnvalue (ScalableFloat|GetValueatLevel _volleysbeforeoverheat))
  (bind _returnvalue_2 (Math|Float|Truncate _returnvalue))
  (bind _shotspervolley (Variables|Default|GetShotsPerVolley))
  (bind _returnvalue_3 (ScalableFloat|GetValueatLevel _shotspervolley))
  (bind _returnvalue_4 (Math|Float|Truncate _returnvalue_3))
  (bind _projectilefireratespawn (Variables|Default|GetProjectileFireRateSpawn))
  (bind _returnvalue_5 (/ 1.0 _projectilefireratespawn))
  ; Start firing
  (Networking|SwitchHasAuthority
    (:Remote
      (Fortnite|FortShippingLog 0 "[RoundFork] Turrets Start Shooting")
      (CallFunction|SetParameters)
      (CallFunction|ToggleFiringProjectiles true)
      (CallFunction|FiringFeedback)
      (bind _returnvalue_6 (Utilities|Time|SetTimerbyEvent (EventDispatchers|CreateEvent :!event Custom|ProjectileFire) _returnvalue_5 true true))
      (Variables|Default|SetProjectileFireTimer _returnvalue_6)
      (bind _returnvalue_7 (Utilities|Time|SetTimerbyEvent (EventDispatchers|CreateEvent :!event Custom|ProjectileImpact) _returnvalue_5 true true (Variables|Default|GetProjectileTravelTime)))
      (Variables|Default|SetProjectileImpactTimer _returnvalue_7)
      ; After N shots, pause firing for a moment and then resume
      (bind _counter (Utilities|FlowControl|DoN _returnvalue_4)
        (:Exit
          (if (== _counter _returnvalue_4)
            (CallFunction|StopFiring)
            (bind _returnvalue_8 (Utilities|Time|SetTimerbyEvent (EventDispatchers|CreateEvent :!event Custom|StartFiring) 4.0))
            (bind _counter_2 (Utilities|FlowControl|DoN _returnvalue_2)
              (:Exit
                (if (== _counter_2 _returnvalue_2)
                  (Utilities|FlowControl|Sequence
                    (:then_0
                      _)
                    (:then_1
                      _))
                  (Utilities|Time|ClearandInvalidateTimerbyHandle _returnvalue_8)
                  (CallFunction|DoOverheat))))))))))

(event Custom|DoOverheat
  ; Overheat
  (Components|Activation|SetActive (Variables|Default|GetNS_RoundFork_Turret_Overheat) true false)
  (Audio|SpawnSoundAttached "/DragonCartNarrative/Assets/Audio/Turret/SC_MatchMist_Turret_Overheat.SC_MatchMist_Turret_Overheat" (Variables|Default|GetTurret))
  (CallFunction|StopFiring))

(event Custom|ProjectileFire
  (bind _volleysbeforeoverheat (Variables|Default|GetVolleysBeforeOverheat))
  (bind _returnvalue (ScalableFloat|GetValueatLevel _volleysbeforeoverheat))
  (bind _returnvalue_2 (Math|Float|Truncate _returnvalue))
  (bind _shotspervolley (Variables|Default|GetShotsPerVolley))
  (bind _returnvalue_3 (ScalableFloat|GetValueatLevel _shotspervolley))
  (bind _returnvalue_4 (Math|Float|Truncate _returnvalue_3))
  (Audio|SpawnSoundAttached "/DragonCartNarrative/Assets/Audio/Turret/SC_MatchMist_Turret_Fire.SC_MatchMist_Turret_Fire" (Variables|Default|GetTurret))
  (CallFunction|FiringFeedback)
  ; After N shots, pause firing for a moment and then resume
  (bind _counter (Utilities|FlowControl|DoN _returnvalue_4)
    (:Exit
      (if (== _counter _returnvalue_4)
        (CallFunction|StopFiring)
        (bind _returnvalue_5 (Utilities|Time|SetTimerbyEvent (EventDispatchers|CreateEvent :!event Custom|StartFiring) 4.0))
        (bind _counter_2 (Utilities|FlowControl|DoN _returnvalue_2)
          (:Exit
            (if (== _counter_2 _returnvalue_2)
              (Utilities|FlowControl|Sequence
                (:then_0
                  _)
                (:then_1
                  _))
              (Utilities|Time|ClearandInvalidateTimerbyHandle _returnvalue_5)
              (CallFunction|DoOverheat))))))))

(event Custom|ProjectileImpact
  (bind _projectileendlocation (Variables|Default|GetProjectileEndLocation))
  (Audio|SpawnSoundatLocation "/DragonCartNarrative/Assets/Audio/Turret/SC_MatchMist_Turret_Projectile_OrbImpact.SC_MatchMist_Turret_Projectile_OrbImpact" _projectileendlocation)
  (Class|BGADCNOrbActor|TriggerRipple (Variables|Default|GetOrbActor) _projectileendlocation (Math|Float|Lerp 6e+03 8e+03 (Math|Random|RandomFloatinRange 0.0 1.0)) (Math|Float|Lerp 0.3 0.5 (Math|Random|RandomFloatinRange 0.0 1.0))))

(event Custom|StopFiring
  ; Stop firing
  (Utilities|Time|ClearandInvalidateTimerbyHandle (Variables|Default|GetProjectileFireTimer))
  (CallFunction|ToggleFiringProjectiles false)
  (Utilities|FlowControl|Delay (+ (Variables|Default|GetProjectileTravelTime) 0.1)) ; Give final projectiles time to hit
  (Components|Activation|SetActive (Variables|Default|GetNS_RoundFork_Turret_Projectiles) false false)
  (Utilities|Time|ClearandInvalidateTimerbyHandle (Variables|Default|GetProjectileImpactTimer)))

(fn FindOrbActor
  (Utilities|IsValid (Variables|Default|GetOrbActor)
    (:Is%20Not%20Valid
      (Fortnite|FortShippingLog 0 "[RoundFork] The Orb actor is not assigned on the turret")
      (bind _outactors (Actor|GetAllActorsOfClass "/DragonCartNarrative/MatchMist/BGA_DCN_OrbActor_MatchMist.BGA_DCN_OrbActor_MatchMist_C"))
      (Utilities|IsValid (Utilities|Array|Get(acopy) _outactors)
        (:Is%20Valid
          (Variables|Default|SetOrbActor (Utilities|Array|Get(acopy) _outactors)))
        (:Is%20Not%20Valid
          (Fortnite|FortShippingLog 0 "[RoundFork] Cannot find an Orb in the map"))))))

(fn FiringFeedback
  (bind _returnvalue (Transformation|GetActorLocation))
  (Rendering|Material|SetScalarParameterValueonMaterials (Variables|Default|GetTurret) "RecoilAnimStartTime" (Utilities|Time|GetGameTimeinSeconds))
  (if (Variables|Default|GetCloseUpTurret)
    (Camera|PlayWorldCameraShake "/Game/Effects/Fort_Effects/BluePrints/CameraShakes/B_Small_Explosion_CameraShake.B_Small_Explosion_CameraShake_C" _returnvalue 2e+02 5e+03)
    (Fortnite|PlayLocalForceFeedbackatLocationMulti "/Game/Weapons/ForceFeedback/Explosions/FF_Generic_ExplosionTiny_Near.FF_Generic_ExplosionTiny_Near" "/Game/Weapons/ForceFeedback/Explosions/FF_Generic_ExplosionTiny_Far.FF_Generic_ExplosionTiny_Far" _returnvalue 2e+03 4e+03)))

(fn SetParameters
  (bind _ns_roundfork_turret_projectiles (Variables|Default|GetNS_RoundFork_Turret_Projectiles))
  (Niagara|SetNiagaraVariable(Position) _ns_roundfork_turret_projectiles "ProjectileEndLocation" (Variables|Default|GetProjectileEndLocation))
  (Effects|Components|ParticleSystem|SetFloatParameter _ns_roundfork_turret_projectiles "ProjectileFireRateSpawn" (ScalableFloat|GetValueatLevel (Variables|Default|GetProjectileFireRateSpawn)))
  (Effects|Components|ParticleSystem|SetFloatParameter _ns_roundfork_turret_projectiles "ProjectileTravelTime" (Variables|Default|GetProjectileTravelTime)))

(fn ToggleFiringProjectiles
  (params (Firing :type Boolean))
  (Effects|Components|ParticleSystem|SetBoolParameter (Variables|Default|GetNS_RoundFork_Turret_Projectiles) "FireProjectiles" Firing)
  (if Firing
    (Components|Activation|SetActive (Variables|Default|GetNS_RoundFork_Turret_Projectiles) true true)
    (Audio|SpawnSoundAttached "/DragonCartNarrative/Assets/Audio/Turret/SC_MatchMist_Turret_Fire.SC_MatchMist_Turret_Fire" (Variables|Default|GetTurret))))

(fn TurretProjectileSetup
  (params (OrbActor :type Actor%20Object%20Reference))
  (bind _returnvalue (Transformation|GetActorLocation OrbActor))
  (bind _turret (Variables|Default|GetTurret))
  (bind _returnvalue_2 (Transformation|GetWorldLocation _turret))
  (bind _returnvalue_3 (Math|Rotator|FindLookatRotation _returnvalue_2 _returnvalue))
  (bind _returnvalue_x (Math|Vector|BreakVector (Transformation|GetActorScale3D OrbActor)))
  (Transformation|SetWorldRotation (Variables|Default|GetTurretBaseMid) (Math|Rotator|MakeRotator 0.0 0.0 (.yaw _returnvalue_3)) false false)
  (Transformation|SetRelativeRotation (Variables|Default|GetTurret) (Math|Rotator|MakeRotator 0.0 (.pitch _returnvalue_3)) false false)
  (bind _output_get (Variables|Default|SetProjectileEndLocation (+ (* (Math|Vector|Normalize (- _returnvalue_2 _returnvalue)) (* _returnvalue_x 5.7e+03)) _returnvalue)))
  (Variables|Default|SetProjectileTravelTime (/ (Math|Vector|Distance(Vector) _output_get (Transformation|GetWorldLocation (Variables|Default|GetTurret))) (ScalableFloat|GetValueatLevel (Variables|Default|GetProjectileSpeed)))))

(fn ConstructionScript
  (bind _orbactor (Variables|Default|GetOrbActor)
    (:then
      (CallFunction|TurretProjectileSetup _orbactor)
      (CallFunction|SetParameters))
    (:else)))
